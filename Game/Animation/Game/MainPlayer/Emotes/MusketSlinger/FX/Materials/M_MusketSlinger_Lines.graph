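# Additive, unlit, two-sided Niagara mesh-particle material that draws animated streak lines for the MusketSlinger emote effects.
# A bright-cored mask across U multiplies two panning Perlin noise samples, time-offset by vertex color red, and a soft fade at both V ends.
# Dynamic parameters drive SplineThicken width and a Sweep mask that grows from V=1 to reveal the line.
# Particle color tints the emissive and scales opacity.
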Material /Game/Animation/Game/MainPlayer/Emotes/MusketSlinger/FX/Materials/M_MusketSlinger_Lines
graph {
 properties {
  bUsedWithNiagaraMeshParticles=true
  blendMode=BLEND_Additive
  pixelDepthOffsetMode=PDOM_Legacy
  shadingModel=MSM_Unlit
  twoSided=true
  usedShadingModels=Unlit
 }
 nodes [
  n1 = DynamicParameter() {parameterIndex=1, defaultValue=<rgba 1 2.5 1 1>, paramNames [Sweep, SplineThickness, Param3, Param4]}
  MaterialFunctionCall(WidthBase=@n1.SplineThickness, WidthTip=@n1.SplineThickness) {materialFunction=&SplineThicken}
  n2 = SphereMask(A=TextureCoordinate() {vTiling=0}, B=Constant2Vector() {r=0.5}, Radius=Constant() {r=0.5}, Hardness=Constant())
  n3 = Add(A=Multiply(A=@n2) {constB=0.5}, B=SmoothStep(Value=@n2) {constMin=0.8})
  n4 = Add(A=Time(), B=VertexColor().R)
  n5 = Panner(Coordinate=TextureCoordinate() {uTiling=0, vTiling=0.25}, Time=@n4) {speedX=0.1, speedY=0.2}
  n6 = Panner(Coordinate=TextureCoordinate() {uTiling=0, vTiling=0.6}, Time=@n4) {speedX=0.1, speedY=0.3}
  n7 = Multiply(
   A=Add(A=TextureSample(UVs=@n5) {texture=/Game/Global/Textures/Default/Noise/3xPerlin_M}.R) {constB=0.25}
   B=Add(A=TextureSample(UVs=@n6) {texture=&3xPerlin_M}.R) {constB=0.25}
  )
  n8 = Multiply(
   A=Saturate(None=Multiply(A=@n7) {constB=2})
   B=SphereMask(A=TextureCoordinate() {uTiling=0}, B=Constant2Vector() {g=0.5}, Radius=Constant() {r=0.5}, Hardness=Constant() {r=0.4})
  )
  n9 = SphereMask(A=TextureCoordinate() {uTiling=0}, B=Constant2Vector() {g=1}, Radius=LinearInterpolate(Alpha=@n1.Sweep) {constB=3}, Hardness=Constant())
  n10 = Saturate(None=Multiply(A=Multiply(A=@n3, B=@n8), B=SmoothStep(Value=@n9) {constMax=0.6, constMin=0.4, constValue=1}))
  n11 = ParticleColor()
  MaterialFunctionCall(Emissive=Multiply(A=@n10, B=@n11.RGB)) {materialFunction=&MF_FX_EmissiveAdjustments}
  Multiply(A=@n10, B=@n11.A)
 ]
}
