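# Additive unlit Niagara mesh-particle material that draws sparkle flares, filed with the Musket Slinger emote's effects.
# A Sprite function turns mesh pieces into sprites, scaled by vertex-color red and dynamic parameters.
# Particle color tints the flare texture through HDR remapping and emissive adjustment.
# Opacity multiplies the flare by particle alpha, per-sprite variation, time-based sine flicker offset by vertex color and instance random, and a near-camera fade skipped on low quality.
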Material /Game/Animation/Game/MainPlayer/Emotes/MusketSlinger/FX/Materials/M_Musket_Slinger_Sparkles
graph {
 properties {
  bUsedWithNiagaraMeshParticles=true
  blendMode=BLEND_Additive
  pixelDepthOffsetMode=PDOM_Legacy
  shadingModel=MSM_Unlit
  usedShadingModels=Unlit
 }
 nodes [
  n1 = VertexColor()
  n2 = DynamicParameter() {paramNames [Scale, FlickerTimeMult, RandomMult, FlickerStrenght]}
  n3 = Multiply(A=@n1.R, B=@n2.RandomMult)
  n4 = LinearInterpolate(
   A=ScalarParameter() {defaultValue=1, parameterName=SparkScale_VC_RangeLow}
   B=ScalarParameter() {defaultValue=1, parameterName=SparkScale_VC_RangeHigh, sortPriority=33}
   Alpha=@n1.R
  ) {constA=0.5}
  MaterialFunctionCall!(Center Location=WorldPosition().XYZ, 0-1 UVs, XY scale=Multiply(A=@n3, B=Multiply(A=@n2.Scale, B=@n4)), OverridePivot, "Normals (Optional)") {materialFunction=&Sprite}
  n5 = ParticleColor()
  n6 = TextureSample() {texture=/Game/Effects/Fort_Effects/Textures/Flares/T_Anime_Flare}
  n7 = Multiply(A=@n5.RGB, B=@n6.G)
  MaterialFunctionCall(Emissive=MaterialFunctionCall(LDR Input=Clamp(None=@n7), HDR Tint=@n7) {materialFunction=&DeriveHDRfromLDR}.RemappedRange) {materialFunction=&MF_FX_EmissiveAdjustments}
  n8 = MaterialFunctionCall(Distance=ScalarParameter() {defaultValue=200, parameterName=CameraDistanceFade}) {materialFunction=&Near_Camera_Fade}
  n9 = Multiply(A=VertexColor().R) {constB=100}
  n10 = Add(A=PerInstanceRandom(), B=Time())
  n11 = Add(A=Multiply(A=Add(A=@n9, B=@n10), B=@n2.FlickerTimeMult), B=@n1.R)
  n12 = Multiply(A=@n5.A, B=LinearInterpolate(B=Sine(None=@n11), Alpha=@n2.FlickerStrenght) {constA=1})
  Multiply(
   A=Multiply(A=Multiply(A=@n6.G, B=@n12), B=@n3)
   B=QualitySwitch(Default=@n8.Result, Low=Constant() {r=1}, High=@n8.Result, Medium=@n8.Result)
  )
 ]
}
